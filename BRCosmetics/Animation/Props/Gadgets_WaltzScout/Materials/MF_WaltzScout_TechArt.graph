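# Material function for the WaltzScout gadget prop that darkens base color with a pencil-mark graphic and writes it back into the material attributes.
# A scaled, offset, clamped line texture supplies the mark in red and an erosion pattern in green, which the PencilMarkAppearance scalar offsets to reveal the mark.
# A TimeStamp value is also built from a timestamp function or looping debug time, while the mark's reveal math reads the scalar instead.
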MaterialFunction /BRCosmetics/Animation/Props/Gadgets_WaltzScout/Materials/MF_WaltzScout_TechArt
graph {nodes [
  n1 = GetMaterialAttributes(None=NamedRerouteUsage()) {attributeGetTypes [#1]}
  FunctionOutput(None=SetMaterialAttributes!(MaterialAttributes=@n1.MaterialAttributes, BaseColor=NamedRerouteUsage()) {attributeSetTypes [#1]})
  MaterialFunctionCall!(
   Time
   Divisor=ScalarParameter() {defaultValue=1.5, group=FNCSShield_Screen, parameterName=DebugTimeDivisor, sortPriority=30}
   Mod=ScalarParameter() {defaultValue=2, group=FNCSShield_Screen, parameterName=DebugTimeModulo, sortPriority=30}
  ) {materialFunction=/Game/Effects/Fort_Effects/Materials/_Functions/MF_ModuloTime}
  NamedRerouteDeclaration(None=FunctionInput() {inputName=In_MaterialAttributes, inputType=FunctionInput_MaterialAttributes}) {name=In_MaterialAttributes, nodeColor=<rgba 0.5938 0 1 1>}
  NamedRerouteDeclaration(None=FunctionInput() {inputName=In_TechArtMask}) {name=In_TechArtMask, nodeColor=<rgba 0 1 0.2734 1>}
  n2 = Divide(A=Time(), B=ScalarParameter() {defaultValue=3, group=FNCSShield_Screen, parameterName=DebugTransitionLength, sortPriority=30})
  n3 = MaterialFunctionCall(
   TimestampA=ScalarParameter() {defaultValue=1, parameterName=Timestamp, sortPriority=30}
   TransitionLength_In=ScalarParameter() {defaultValue=0.1, parameterName=TransitionLength, sortPriority=30}
  ) {materialFunction=&MF_Timestamp_InOut_01}
  NamedRerouteDeclaration(None=StaticSwitchParameter(True=Frac(None=@n2), False=VertexInterpolator(VS=@n3.Linear).PS) {group=FNCSShield_Screen_Graphics, parameterName=DebugModeOn, sortPriority=0}) {name=TimeStamp, nodeColor=<rgba 0.125 1 0 1>}
  n4 = TextureCoordinate()
  n5 = AppendVector(
   A=ScalarParameter() {defaultValue=4, group=WaltzScout_TechArt, parameterName=PencilMarkScaleX}
   B=ScalarParameter() {defaultValue=2, group=WaltzScout_TechArt, parameterName=PencilMarkScaleY}
  )
  n6 = Multiply(
   A=ScalarParameter() {defaultValue=-1.6144, group=WaltzScout_TechArt, parameterName=PencilMarkOffsetX}
   B=ScalarParameter() {defaultValue=4, group=WaltzScout_TechArt, parameterName=PencilMarkScaleX}
  )
  n7 = Multiply(
   A=ScalarParameter() {defaultValue=-0.328, group=WaltzScout_TechArt, parameterName=PencilMarkOffsetY}
   B=ScalarParameter() {defaultValue=2, group=WaltzScout_TechArt, parameterName=PencilMarkScaleY}
  )
  n8 = TextureSample(UVs=Add(A=Multiply(A=@n4, B=@n5), B=AppendVector(A=@n6, B=@n7))) {
   samplerSource=SSM_Clamp_WorldGroupSettings
   samplerType=SAMPLERTYPE_LinearColor
   texture=/BRCosmetics/Animation/Props/Gadgets_WaltzScout/Textures/T_WaltzScout_Gadget_Line
  }
  n9 = ScalarParameter() {defaultValue=2, parameterName=PencilMarkAppearance}
  n10 = Add(A=NamedRerouteDeclaration(None=@n8.G) {name=PencilMarkErosion, nodeColor=<rgba 1 0 0.1172 1>}, B=Subtract(A=Multiply(A=@n9) {constB=2}))
  NamedRerouteDeclaration(None=Multiply(A=NamedRerouteDeclaration(None=@n8.R) {name=PencilMark, nodeColor=<rgba 0 1 0.2031 1>}, B=Saturate(None=@n10))) {name=TimeStampedPencilMark, nodeColor=<rgba 0 0.0859 1 1>}
  NamedRerouteDeclaration(None=@n1.BaseColor) {name=In_BaseColor, nodeColor=<rgba 0.4062 1 0 1>}
  NamedRerouteDeclaration(None=LinearInterpolate(A=NamedRerouteUsage(), Alpha=NamedRerouteUsage()) {constB=0.05}) {name=Out_BaseColor, nodeColor=<rgba 1 0 0.7266 1>}
  OneMinus(None=NamedRerouteUsage())
 ]}
